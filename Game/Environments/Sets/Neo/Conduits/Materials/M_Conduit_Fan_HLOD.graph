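# Masked, two-sided Default Lit material for the HLOD version of the Neo conduit fan, usable on instanced static meshes.
# It samples a custom LOD diffuse texture and uses its red channel to blend paired brightness, roughness and metallic parameters.
# The blended brightness multiplies the diffuse color, alongside a scalar specular parameter and a custom LOD normal map.
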Material /Game/Environments/Sets/Neo/Conduits/Materials/M_Conduit_Fan_HLOD
graph {
 properties {bUsedWithInstancedStaticMeshes=true, blendMode=BLEND_Masked, pixelDepthOffsetMode=PDOM_Legacy, twoSided=true, usedShadingModels=Default Lit}
 nodes [
  n1 = TextureSample() {texture=/Game/Environments/Sets/Neo/Conduits/Textures/T_CustomLOD_Fan_D}
  LinearInterpolate(
   A=ScalarParameter() {defaultValue=0.1143, parameterName=Roughness2}
   B=ScalarParameter() {defaultValue=2.0391, parameterName=Roughness}
   Alpha=@n1.R
  )
  LinearInterpolate(
   A=ScalarParameter() {defaultValue=1.7614, parameterName=Metallic2}
   B=ScalarParameter() {defaultValue=-1.9869, parameterName=Metallic}
   Alpha=@n1.R
  )
  Multiply(
   A=LinearInterpolate(
    A=ScalarParameter() {defaultValue=1.5175, parameterName=Brightnes2}
    B=ScalarParameter() {defaultValue=1.3865, parameterName=Brightnes}
    Alpha=@n1.R
   )
   B=@n1.RGB
  )
  ScalarParameter() {defaultValue=0.4567, parameterName=Specular}
  TextureSample() {samplerType=SAMPLERTYPE_Normal, texture=/Game/Environments/Sets/Neo/Conduits/Textures/T_CustomLOD_Fan_N}
 ]
}
