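# Surface material for the Neo Tilted conduit fan, driven by the global SlipstreamsAlpha parameter between stopped and running looks.
# World position offset spins red vertex-colored parts around a configurable axis and pivot at a speed scaled by that alpha, with corrected normals.
# When off, specular-alpha-masked areas turn dark gray and their blink term multiplies to black.
# When on, they glow in time-of-day colors; a static switch substitutes a custom emissive texture.
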Material /Game/Environments/Sets/Neo/Conduits/Materials/M_NeoTilted_Conduits_Fan
graph {
 properties {pixelDepthOffsetMode=PDOM_Legacy, usedShadingModels=Default Lit}
 nodes [
  n1 = CollectionParameter() {collection=&FortniteMaterialParameters, parameterName=SlipstreamsAlpha}
  n2 = LinearInterpolate(A=Constant(), B=ScalarParameter() {defaultValue=-0.7, parameterName=RotateSpeed}, Alpha=@n1)
  n3 = Transform(None=VectorParameter() {parameterName=RotationAxis, defaultValue=<rgba 0 1 0 1>}.RGB) {transformSourceType=TRANSFORMSOURCE_Local}
  n4 = Multiply(A=Time(), B=@n2) {constB=0.2}
  n5 = Add(
   A=MaterialFunctionCall() {materialFunction=&ObjectPivotPoint}.Object Pivot Location
   B=Transform(None=VectorParameter() {parameterName=PivotOffset, defaultValue=<rgba 0 0 772 0>}.RGB) {transformSourceType=TRANSFORMSOURCE_Local}
  )
  n6 = WorldPosition()
  n7 = Constant()
  n8 = VectorParameter() {defaultValue=<rgba 0 0 100 0>}
  n9 = Multiply(
   A=VertexColor().R
   B=Add(
    A=RotateAboutAxis(NormalizedRotationAxis=@n3, RotationAngle=@n4, PivotPoint=@n5, Position=@n6.XYZ)
    B=PreviousFrameSwitch(Current Frame=@n7, Previous Frame=@n8.RGB)
   )
  )
  If(A=@n2, B=Constant(), "A > B"=@n9, "A == B"=Constant(), "A < B"=@n9)
  n10 = TextureSampleParameter2D() {parameterName=Diffuse, texture=/Game/Environments/Sets/Neo/Conduits/Textures/T_NeoTilted_Conduits_Fan_D}
  n11 = Desaturation(None=@n10.RGB, Fraction=Constant() {r=1})
  n12 = Constant3Vector() {constant=<rgba 0.3073 0.3073 0.3073 0>}
  n13 = TextureSampleParameter2D() {parameterName=Specular, samplerType=SAMPLERTYPE_Masks, texture=/Game/Environments/Sets/Neo/Conduits/Textures/T_NeoTilted_Conduits_Fan_S}
  n14 = Multiply(A=@n13.A) {constB=4}
  n15 = OneMinus(None=LinearInterpolate(A=Constant() {r=1}, B=Constant(), Alpha=Clamp(None=@n14)) {constB=0})
  n16 = CollectionParameter() {collection=&FortniteMaterialParameters, parameterName=SlipstreamsAlpha}
  LinearInterpolate(A=LinearInterpolate(A=@n10.RGB, B=Multiply(A=@n11, B=@n12), Alpha=@n15), B=@n10.RGB, Alpha=@n16)
  n17 = TextureSampleParameter2D() {parameterName=Normals, samplerType=SAMPLERTYPE_Normal, texture=/Game/Environments/Sets/Neo/Conduits/Textures/T_NeoTilted_Conduits_Fan_N}
  MaterialFunctionCall(
   BaseNormal=MaterialFunctionCall(Rotation Axis=@n3, Rotation Angle=@n4, Movement Mask=VertexColor().R) {materialFunction=&FixRotateAboutAxisNormals}.New Tangent Space Vertex Normal
   AdditionalNormal=@n17.RGB
  ) {materialFunction=&BlendAngleCorrectedNormals}
  n18 = Sine(None=Multiply(A=Time(), B=ScalarParameter() {parameterName=Disabled Blink Speed}))
  Multiply(A=@n12, B=Clamp(None=ConstantBiasScale(None=@n18)))
  n19 = VectorParameter() {group=Emissive, parameterName=Day Color, defaultValue=<rgba 0 0 0 1>}
  n20 = MaterialFunctionCall(
   Color_Morning=@n19.RGB
   Color_Day=@n19.RGB
   Color_Evening=@n19.RGB
   Color_Night=VectorParameter() {group=Emissive, parameterName=Night Color, defaultValue=<rgba 1.8415 0.1637 2 1>}.RGB
  ) {materialFunction=&Color_Time_of_Day}
  n21 = ConstantBiasScale(None=@n18) {bias=1.05}
  n22 = Multiply(A=Multiply(A=Constant3Vector(), B=Clamp(None=@n21)), B=@n13.A)
  n23 = CollectionParameter() {collection=&FortniteMaterialParameters, parameterName=SlipstreamsAlpha}
  StaticSwitchParameter(
   True=Multiply(
    A=@n20.Result
    B=TextureSampleParameter2D() {parameterName=Emissive, texture=/Game/Environments/Artemis/Sets/Celebration/Textures/T_Celebration_Conduits_Fan_E}.RGB
   )
   False=LinearInterpolate(A=Multiply(A=@n15, B=@n22), B=Multiply(A=@n20.Result, B=@n13.A), Alpha=@n23)
  ) {parameterName=Use Custom Emissive?}
 ]
}
